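# Default Lit rock material for a debris field vortex, moving mesh pieces in the vertex shader.
# World position offset spins vertices around a parameterized axis at the actor position, with speed scaled by vertex color red and green.
# It adds push-out from center, normal-based inflation, and switchable curl-noise drift and camera-distance-scaled sprites.
# Rock textures drive shading, with optional vertex-red darkening and normals corrected for the rotation.
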Material /Game/Effects/Fort_Effects/World/Destruction/Rocks/M_DebrisField_Vortex
graph {
 properties {pixelDepthOffsetMode=PDOM_Legacy, usedShadingModels=Default Lit}
 nodes [
  n1 = VectorParameter() {parameterName=RotateAxis, defaultValue=<rgba 0 0 1 0>}
  n2 = MaterialFunctionCall() {materialFunction=&ObjectScale}
  n3 = Divide(A=@n1.RGB, B=@n2.Scale XYZ)
  n4 = VertexColor()
  n5 = Time()
  n6 = Multiply(A=Multiply(A=@n4.G, B=Multiply(A=@n4.R, B=@n5)), B=ScalarParameter() {defaultValue=0.145, parameterName=Speed})
  n7 = ActorPositionWS()
  n8 = ScalarParameter() {defaultValue=0.3, parameterName=Slow}
  n9 = MaterialFunctionCall(Vector 3=@n2.Scale XYZ) {materialFunction=/Engine/Functions/Engine_MaterialFunctions02/Utility/VectorLength}
  n10 = Add(
   A=Divide(A=WorldPosition() {worldPositionShaderOffset=WPT_ExcludeAllShaderOffsets}.XYZ) {constB=8000}
   B=Multiply(A=Time() {period=16}, B=Multiply(A=@n8, B=@n9.V3 Length))
  )
  n11 = ScalarParameter() {defaultValue=23, parameterName=ScaleOffset}
  n12 = Constant3Vector()
  n13 = StaticSwitchParameter(True=Multiply(A=VectorNoise(World Position=@n10) {noiseFunction=VNF_CurlALU}, B=Multiply(A=@n11, B=@n9.V3 Length)), False=@n12) {parameterName=UseOffset}
  n14 = Add(
   A=RotateAboutAxis(
    NormalizedRotationAxis=Transform(None=@n3) {transformSourceType=TRANSFORMSOURCE_Local}
    RotationAngle=@n6
    PivotPoint=@n7
    Position=WorldPosition().XYZ
   )
   B=@n13
  )
  n15 = ScalarParameter() {defaultValue=800, parameterName=SoftenDistance}
  n16 = Multiply(
   A=@n9.V3 Length
   B=LinearInterpolate(
    A=ScalarParameter() {defaultValue=1, parameterName=SpriteScaleNear}
    B=ScalarParameter() {defaultValue=10, parameterName=SpriteScaleFar}
    Alpha=MaterialFunctionCall(Distance=@n15) {materialFunction=&Near_Camera_Fade}.Result
   )
  )
  n17 = StaticSwitchParameter(
   True=MaterialFunctionCall!(
    Center Location=WorldPosition() {worldPositionShaderOffset=WPT_ExcludeAllShaderOffsets}.XYZ
    0-1 UVs
    XY scale=AppendVector(A=@n16, B=@n16)
    OverridePivot
    "Normals (Optional)"
   ) {materialFunction=&Sprite}.World position offset
   False=@n12
  ) {parameterName=UseSprite}
  n18 = ActorPositionWS()
  n19 = Multiply(A=WorldPosition().XYZ, B=Constant3Vector() {constant=<rgba 0 0 1 0>})
  n20 = Normalize(VectorInput=Subtract(A=WorldPosition().XYZ, B=Add(A=@n18, B=@n19)))
  n21 = ScalarParameter() {defaultValue=500, parameterName=PushOutAwayFromCenter}
  Add(A=Add(A=Add(A=@n14, B=@n17), B=Multiply(A=@n20, B=@n21)), B=Multiply(A=VertexNormalWS(), B=ScalarParameter() {parameterName=BigRockWPO}))
  Add(A=@n13, B=@n17)
  Constant3Vector() {constant=<rgba 1 0 0 0>}
  n22 = VectorParameter() {parameterName=BaseColor, defaultValue=<rgba 1 1 1 0>}
  n23 = Clamp(None=Multiply(A=VertexColor().R) {constB=1.2})
  n24 = Multiply(A=TextureCoordinate(), B=ScalarParameter() {defaultValue=1, parameterName=TileUVs})
  n25 = TextureSampleParameter2D(UVs=@n24) {parameterName=DiffuseTexture, texture=/Game/Environments/Props/Rocks/Textures/T_Props_Rocks_D}
  n26 = StaticSwitchParameter(True=@n25.R, False=@n25.R) {parameterName=UseR_ForDiffuse}
  LinearInterpolate(A=Multiply(A=@n22.RGB, B=Multiply(A=@n23, B=@n26)), B=Multiply(A=@n22.RGB, B=@n26), Alpha=ScalarParameter() {parameterName=UseVertexColor})
  MaterialFunctionCall(
   Rotation Axis=@n1.RGB
   Rotation Angle=@n6
   World Space Vertex Normals=Transform(None=TextureSampleParameter2D(UVs=@n24) {parameterName=NormalTexture, samplerType=SAMPLERTYPE_Normal, texture=/Game/Environments/Props/Rocks/Textures/T_Props_Rocks_N}.RGB)
   Normals Pivot=@n7
  ) {materialFunction=&FixRotateAboutAxisNormals}
  Multiply(
   A=TextureSampleParameter2D(UVs=@n24) {parameterName=SpecTexture, samplerType=SAMPLERTYPE_Masks, texture=/Game/Environments/Props/Rocks/Textures/T_Props_Rocks_S}.RGB
   B=ScalarParameter() {defaultValue=1, parameterName=Spec Mult, sliderMax=1}
  )
  Multiply(A=VertexNormalWS())
  StaticSwitchParameter(True=@n25.G, False=@n25.G) {parameterName=UseG_ForOpacityMask}
 ]
}
